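# Sound cue for the destruction of the Boombox gadget, built from a single wave player on Boombox_Destroy_01.
# It plays at reduced volume in the melee weapon sound class and references the shared Boombox effects attenuation asset.
# Inline attenuation values for distance low-pass and occlusion are stored, but the override flag is not enabled, so the referenced asset governs spatialization.
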SoundCue /Game/Sounds/Fort_Gadgets/Boombox/Boombox_Destroy_Cue
root=WavePlayer_0
nodes [{id=WavePlayer_0, properties {soundWaveAssetPtr=&/Boombox_Destroy_01}}]
properties {
 attenuationSettings=&/Boombox_Effects_Att
 maxDistance=6000
 soundClassObject=&Fort_WeaponMelee_PSM
 volumeMultiplier=0.4
 attenuationOverrides {
  bAttenuateWithLPF=true
  bEnableOcclusion=true
  falloffDistance=4000
  lPFFrequencyAtMax=1000
  lPFRadiusMax=3000
  lPFRadiusMin=1000
  occlusionInterpolationTime=1
  occlusionLowPassFilterFrequency=1000
  occlusionTraceChannel=ECC_GameTraceChannel11
  occlusionVolumeAttenuation=0.6
  stereoSpread=2000
  attenuationShapeExtents {x=0}
 }
}
